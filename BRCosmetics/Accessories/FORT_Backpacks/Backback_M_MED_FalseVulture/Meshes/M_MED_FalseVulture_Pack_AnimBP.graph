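# Animation Blueprint for the FalseVulture back bling, adjusting pack bones to follow the wearer and toggling part of its material.
# Owner spine deltas drive pack roll and offset via chained additive Modify Bone nodes, alongside fixed pack and sword offsets.
# A smoothed melee-animation weight, scaled by head-down alpha, blends in extra roll from the left upper arm.
# Weapon, customization and skydiving events set "Hide Element 01" when a tagged FalseVulture pickaxe is equipped outside skydiving.
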AnimBlueprint /BRCosmetics/Accessories/FORT_Backpacks/Backback_M_MED_FalseVulture/Meshes/M_MED_FalseVulture_Pack_AnimBP
target_skeleton=&/M_MED_FalseVulture_Pack_Skeleton
graphs [
 AnimGraph {nodes [
   n1 = "Transform (Modify) Bone - Bone: pack"(ComponentPose=Local To Component(LocalPose=Input Pose()), Translation="-5,-2,0", Rotation="0,0,0") {
    translationMode=BMM_Additive
    translationSpace=BCS_BoneSpace
    alphaBoolBlend {blendOption=Linear}
    boneToModify {boneName=pack}
    translation=<-5 -2>
   }
   n2 = "Transform (Modify) Bone - Bone: pack"(
    ComponentPose=@n1
    Translation=Get Backpack Translation() -> (Backpack Translation:Vector)
    Rotation=Get Backpack Rotate() -> (Backpack Rotate:Rotator)
   ) {
    rotationMode=BMM_Additive
    rotationSpace=BCS_BoneSpace
    translationMode=BMM_Additive
    translationSpace=BCS_BoneSpace
    alphaBoolBlend {blendOption=Linear}
    boneToModify {boneName=pack}
   }
   n3 = "Transform (Modify) Bone - Bone: base_sword"(ComponentPose=@n2, Alpha=1, Translation="0,0,0", Rotation="0,10,0", Scale="1,1,1") {
    rotationMode=BMM_Additive
    rotationSpace=BCS_BoneSpace
    translationMode=BMM_Additive
    translationSpace=BCS_BoneSpace
    alphaBoolBlend {blendOption=Linear}
    boneToModify {boneName=base_sword}
    rotation {yaw=10}
   }
   n4 = "Transform (Modify) Bone - Bone: pack"(
    ComponentPose=@n3
    Alpha=Get isMelee() -> (isMelee:double)
    Translation="0,-1.5,0"
    Rotation=Get Backpack Rotate Pickaxe() -> (Backpack Rotate Pickaxe:Rotator)
   ) {
    rotationMode=BMM_Additive
    rotationSpace=BCS_BoneSpace
    translationMode=BMM_Additive
    translationSpace=BCS_BoneSpace
    alphaBoolBlend {blendOption=Linear}
    boneToModify {boneName=pack}
    rotation {roll=20}
    translation {y=-1.5}
   }
   Output Pose(Result=Component To Local(ComponentPose=@n4)) {layerGroup=None}
  ]}
 EventGraph {nodes [
   n1 = Event BlueprintUpdateAnimation()
   n2 = GetOwnerSkeletalMeshComponent()
   n3 = BreakRotator(InRot=BreakTransform(InTransform=GetDeltaTransformFromRefPose(self=@n2, BoneName=spine_05, BaseName=spine_01)).Rotation)
   > Set Backpack Rotate(Backpack Rotate:Rotator=MakeRotator(Roll=MapRangeClamped(Value=@n3.Yaw, InRangeA=0, InRangeB=-40, OutRangeB=30))) -> (Output_Get:Rotator)
   > Set Backpack Translation(Backpack Translation:Vector=MakeVector(Y=MapRangeClamped(Value=@n3.Yaw, InRangeA=0, InRangeB=-30, OutRangeB=-5))) -> (Output_Get:Vector)
   n4 = GetOwnerSkeletalMeshComponent()
   > n5 = Set head down rot(head down rot:double=GetAlphaFromReferenceTransformDelta(SkeletalMeshComponent=@n4, BoneName=spine_01, BaseName=head, InputRange=(X=0,Y=-45), TwistAxis="1,0,0")) -> (Output_Get:double)
   n6 = "To Float (Boolean)"(InBool=Get bIsPlayingMeleeAnim())
   n7 = float * float(A=@n1.DeltaTimeX, B=0.2) {is=title K2Node_PromotableOperator}
   n8 = Get isMelee()
   > Set isMelee(isMelee:double=float * float(A=@n5, B=FInterpTo(Current=@n8, Target=@n6, DeltaTime=@n7, InterpSpeed=2))) -> (Output_Get:double)
   n9 = BreakRotator(InRot=BreakTransform(InTransform=GetDeltaTransformFromRefPose(self=@n4, BoneName=upperarm_l, BaseName=clavicle_l)).Rotation)
   > Set Backpack Rotate Pickaxe(Backpack Rotate Pickaxe:Rotator=MakeRotator(Roll=MapRangeClamped(Value=@n9.Yaw, InRangeA=0, InRangeB=90, OutRangeB=40))) -> (Output_Get:Rotator)
   n10 = Event OnWeaponChanged()
   n11 = Get Dynamic Material Instance(execute=[@n10.then, @n17.then], SkeletalMeshComponent=GetPartSkeletalMeshComponent())
   n12 = PickaxeItemDefinitionHasAnyMetaTag(Tags=Make Array("[0]"=(TagName="Cosmetics.MetaTags.FalseVulture")), bRequiresEquipped=true)
   > n13 = Branch(Condition=@n12)
   > n14 = Branch(Condition=Get bIsSkyDiving())
   SetScalarParameterValue(execute=[@n15.then, @n14.then, @n16.else, @n13.else], self=@n11, ParameterName=Hide Element 01, Value=0)
   SetScalarParameterValue(execute=[@n14.else, @n16.then], self=@n11, ParameterName=Hide Element 01, Value=1)
   n15 = Event OnBeginSkydiving()
   Event OnEndSkydiving()
   > n16 = Branch(Condition=@n12)
   n17 = Event OnCharacterCustomizationLoaded()
   TryGetPawnOwner()
  ]}
]
variables [
 Backpack Rotate {type=Rotator}
 Backpack Rotate Pickaxe {type=Rotator}
 Backpack Translation {type=Vector}
 head down rot {type=number}
 isMelee {type=number}
]
